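; Balloon gameplay actor for the Athena consumable, covering its look, team, popping and release.
; BeginPlay applies color and stencil through a dynamic material, follows owner team changes on authority and adjusts animation budgeting.
; DoOnce-gated damage sends quest events and pops it with FX; death from damage credits the shooter as fall instigator.
; Release detaches it to fly via projectile movement, then pops and destroys it; a separate event pops it at max height.
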(event EventBeginPlay
  (bind _returnvalue (Actor|GetOwner))
  (bind _balloon_color_selection (Variables|Default|GetBalloon%20Color%20Selection))
  (bind _balloon_color_choices (Variables|Default|GetBalloon%20Color%20Choices))
  (bind _output (Utilities|Array|Get(aref) _balloon_color_choices _balloon_color_selection))
  (bind _balloon_skbudgeted (Variables|Default|GetBalloon_SKBudgeted))
  ;[1 1 1] Begin play and set the materials on the balloon
  (Networking|SwitchHasAuthority
    (:Authority
      ;[1 1 1] Make sure we have an event for when players switch teams.
      (bind _asfort_pawn (Utilities|Casting|CastToFortPawn (Actor|GetOwner)))
      (bind _asfort_player_state (Utilities|Casting|CastToFortPlayerState (Variables|/Script/Engine.Pawn|GetPlayerState _asfort_pawn)))
      (Player|BindEventtoOnTeamIndexChangedDelegate _asfort_player_state (EventDispatchers|CreateEvent :!event Custom|OnOwnerTeamIndexChanged)))
    (:Remote
      _))
  (bind _returnvalue_2 (Rendering|Material|CreateDynamicMaterialInstance (Variables|Default|GetBalloon_SKBudgeted) 0 (select (Fortnite|IsCreativeorPlaygroundModePlaylist) "/Game/Weapons/Prototype/Flotation_Balloon/Material/Flotation_Balloon.Flotation_Balloon" "/Game/Weapons/Prototype/Flotation_Balloon/Material/Flotation_Balloon.Flotation_Balloon") "None" :!class /Script/Engine.PrimitiveComponent))
  (bind _output_get (Variables|Default|SetBalloonMID _returnvalue_2))
  (Rendering|Material|SetMaterial (Variables|Default|GetBalloon_SKBudgeted) 0 _output_get)
  (Rendering|Material|SetVectorParameterValue _output_get "Balloon Color" _output :!class /Script/Engine.MaterialInstanceDynamic)
  (Rendering|Material|SetTextureParameterValue _output_get "Stencil" (Utilities|Array|Get(acopy) (Variables|Default|GetStencilArray) (Variables|Default|GetBalloonStencilSelection)))
  (Variables|Default|SetBalloon_Color_Cache _output)
  (Utilities|FlowControl|Delay 0.4)
  (Variables|/Script/Engine.SkinnedMeshComponent|SetVisibilityBasedAnimTickOption "OnlyTickPoseWhenRendered" _balloon_skbudgeted)
  (Components|Animation|SetAnimationMode _balloon_skbudgeted "AnimationBlueprint" true)
  (if (Pawn|IsLocallyControlled (Utilities|Casting|CastToPawn _returnvalue))
    (CallFunction|UnregisterMeshfromAnimationBudgetAllocator _balloon_skbudgeted))
  (Quest|SendComplexCustomStatEvent (Quests|GetQuestManager (Fortnite|GetFortPlayerControllerfromActor _returnvalue) "Invalid") 0 "(GameplayTags=((TagName=\"Athena.Quests.Balloons\")))" "(GameplayTags=)" 1))

(event Custom|OnOwnerTeamIndexChanged
  ;[1 1 1] Switch teams to match the player's new team.
  (bind _asfort_pawn (Utilities|Casting|CastToFortPawn (Actor|GetOwner)))
  (CallFunction|SetTeam (Team|GetActorTeam _asfort_pawn)))

(event Custom|BroadcastBalloonShotEvent
  (params (Shooter :type Actor%20Object%20Reference))
  (Ability|SendGameplayEventtoActor (Game|GetInstigator) "(TagName=\"Event.Athena.Balloons.Shot\")" (Utilities|Struct|MakeGameplayEventData "(TagName=\"Event.Athena.Balloons.Shot\")" Shooter self)))

(event Custom|OpenGate
  (Utilities|FlowControl|DoOnce false
    (:Completed
      (if (> (stale EventOnDamageServer Damage) 0.0)
        (Ability|SendGameplayEventtoActor (stale EventOnDamageServer InstigatedBy) "(TagName=\"Event.Athena.Balloons.Destroyed\")" (Utilities|Struct|MakeGameplayEventData "(TagName=\"Event.Athena.Balloons.Destroyed\")" (stale EventOnDamageServer InstigatedBy) (stale EventOnDamageServer InstigatedBy))) ; Separate from "Shot" for quest context of removing your own or shooting other player's balloons
        (Creative|GameplayEffects|SimpleApplyGameplayEffectToActor "/Game/Athena/Items/Consumables/Balloons/GE_Athena_Balloons_Consumable_QuestCheck_Destroyed.GE_Athena_Balloons_Consumable_QuestCheck_Destroyed_C" (stale EventOnDamageServer InstigatedBy))
        (Utilities|FlowControl|DoOnce false
          (:Completed
            (CallFunction|BroadcastBalloonShotEvent (stale EventOnDamageServer DamageCauser))
            (CallFunction|PlayPopFX (stale EventOnDamageServer HitInfo))))
        (else
          (CallFunction|PlayTeamBalloonImpactSound)
          (CallFunction|OpenGate))))) ; prevents spread weapons from overloading damage
)

(event EventOnDamageServer
  (params (Damage :type Float%20%28single-precision%29) (DamageTags :type Gameplay%20Tag%20Container%20Structure%20%28by%20ref%29) (Momentum :type Vector :default "0, 0, 0") (HitInfo :type Hit%20Result%20Structure%20%28by%20ref%29) (InstigatedBy :type Controller%20Object%20Reference) (DamageCauser :type Actor%20Object%20Reference) (EffectContext :type Gameplay%20Effect%20Context%20Handle%20Structure))
  (Utilities|FlowControl|DoOnce false
    (:Completed
      (if (> Damage 0.0)
        (Ability|SendGameplayEventtoActor InstigatedBy "(TagName=\"Event.Athena.Balloons.Destroyed\")" (Utilities|Struct|MakeGameplayEventData "(TagName=\"Event.Athena.Balloons.Destroyed\")" InstigatedBy InstigatedBy)) ; Separate from "Shot" for quest context of removing your own or shooting other player's balloons
        (Creative|GameplayEffects|SimpleApplyGameplayEffectToActor "/Game/Athena/Items/Consumables/Balloons/GE_Athena_Balloons_Consumable_QuestCheck_Destroyed.GE_Athena_Balloons_Consumable_QuestCheck_Destroyed_C" InstigatedBy)
        (Utilities|FlowControl|DoOnce false
          (:Completed
            (CallFunction|BroadcastBalloonShotEvent DamageCauser)
            (CallFunction|PlayPopFX HitInfo)))
        (else
          (CallFunction|PlayTeamBalloonImpactSound)
          (CallFunction|OpenGate))))) ; prevents spread weapons from overloading damage
)

(event EventOnDeathServer
  (params (Damage :type Float%20%28single-precision%29) (DamageTags :type Gameplay%20Tag%20Container%20Structure%20%28by%20ref%29) (Momentum :type Vector :default "0, 0, 0") (HitInfo :type Hit%20Result%20Structure%20%28by%20ref%29) (InstigatedBy :type Controller%20Object%20Reference) (DamageCauser :type Actor%20Object%20Reference) (EffectContext :type Gameplay%20Effect%20Context%20Handle%20Structure))
  ;[1 1 1] When this item dies from being shot, remove its collision
  (if (Math|Conversions|ToBoolean(Integer) (Math|Float|Truncate Damage))
    (Development|PrintString "shot" false false)
    (Collision|SetCollisionEnabled (Variables|Default|GetBalloon_SKBudgeted) "NoCollision")
    (bind _asfort_player_pawn_athena (Utilities|Casting|CastToFortPlayerPawnAthena (Variables|Actor|GetInstigator)))
    (Class|FortPlayerPawnAthena|SetFallInstigator _asfort_player_pawn_athena InstigatedBy "(GameplayTags=((TagName=\"Gameplay.Damage.Environment.Falling.Balloons\")))")
    (else
      (Development|PrintString "release pop"))))

(event Custom|PlayTeamBalloonImpactSound
  (Audio|SpawnSoundAttached (Variables|Default|GetBalloonIndestruct_Sound) (Variables|Default|GetBalloon_SKBudgeted)))

(event Custom|PlayPopFX
  (params (HitEvent :type Hit%20Result%20Structure))
  (bind _balloon_skbudgeted (Variables|Default|GetBalloon_SKBudgeted))
  ;[1 1 1] Play death VFX
  (Audio|SpawnSoundAttached (Variables|Default|GetBalloonShotCue) _balloon_skbudgeted)
  (Development|SetHiddeninGame (Variables|Default|GetBalloon_SKBudgeted) true false)
  (bind _returnvalue (Effects|Components|ParticleSystem|SpawnFXAtLocation "/Game/Athena/Items/Consumables/Balloons/FXs/NS_Balloons_Pop.NS_Balloons_Pop" (Transformation|GetSocketLocation _balloon_skbudgeted "balloon_center_socket")))
  (Effects|Components|ParticleSystem|SetColorParameter _returnvalue "Balloon Color" (Variables|Default|GetBalloon_Color_Cache)))

(event Custom|PopAtMaxHeight
  (Collision|SetCollisionEnabled (Variables|Default|GetBalloon_SKBudgeted) "NoCollision")
  (Networking|CanExecuteCosmeticEvents
    (:True
      (Audio|SpawnSoundAttached (Variables|Default|GetBalloonPopCue) (Variables|Default|GetBalloon_SKBudgeted))
      ;[1 1 1] Play death VFX
      (Development|SetHiddeninGame (Variables|Default|GetBalloon_SKBudgeted) true false)
      (bind _returnvalue (Effects|Components|ParticleSystem|SpawnFXAtLocation "/Game/Athena/Items/Consumables/Balloons/FXs/NS_Balloons_Pop.NS_Balloons_Pop" (Transformation|GetSocketLocation (Variables|Default|GetBalloon_SKBudgeted) "balloon_center_socket")))
      (Effects|Components|ParticleSystem|SetColorParameter _returnvalue "Balloon Color" (Variables|Default|GetBalloon_Color_Cache)))))

(event Custom|Released
  (bind _returnvalue (Game|GetInstigator))
  (bind _balloonmid (Variables|Default|GetBalloonMID))
  (bind _balloon_skbudgeted (Variables|Default|GetBalloon_SKBudgeted))
  ;[1 1 1] When released, this function is called from the weapon to cause the balloon to fly into the air and have its material parameters set for VFX
  (Transformation|DetachFromActor "KeepWorld" "KeepRelative" "KeepWorld")
  (Collision|SetCollisionProfileName (Variables|Default|GetBalloon_SKBudgeted) "FortProjectileBall" true :!class /Script/Engine.PrimitiveComponent)
  (Variables|Balloons|SetBalloonNoAim 0.0)
  (Variables|/Script/Engine.ProjectileMovementComponent|SetbSimulationEnabled true (Variables|Default|GetProjectileMovement))
  (Variables|/Script/Engine.MovementComponent|SetVelocity (+ (Variables|/Script/Engine.MovementComponent|GetVelocity (Variables|Default|GetProjectileMovement)) (Transformation|GetVelocity (Game|GetInstigator))) (Variables|Default|GetProjectileMovement))
  (Rendering|Material|SetScalarParameterValue _balloonmid "WiggleStrength" (Math|Interpolation|FInterpTo 1.0 2e+01 (Utilities|Time|GetGameTimeinSeconds) 1.25) :!class /Script/Engine.MaterialInstanceDynamic)
  (Rendering|Material|SetScalarParameterValue _balloonmid "TimeScale" (Math|Interpolation|FInterpTo 1.0 3.0 (Utilities|Time|GetGameTimeinSeconds) 1.25) :!class /Script/Engine.MaterialInstanceDynamic)
  (Utilities|FlowControl|Delay 2.0)
  ;[0.05729 0.05729 0.05729 0.5] Using the same event & tag data as when shot for quest progression of "When a balloon is shot or released"
  (Ability|SendGameplayEventtoActor _returnvalue "(TagName=\"Event.Athena.Balloons.Destroyed\")" (Utilities|Struct|MakeGameplayEventData "(TagName=\"Event.Athena.Balloons.Destroyed\")" _returnvalue self))
  (Creative|GameplayEffects|SimpleApplyGameplayEffectToActor "/Game/Athena/Items/Consumables/Balloons/GE_Athena_Balloons_Consumable_QuestCheck_Destroyed.GE_Athena_Balloons_Consumable_QuestCheck_Destroyed_C" _returnvalue)
  (Development|SetHiddeninGame _balloon_skbudgeted true false)
  (Components|Activation|Activate (Variables|Default|GetAudio) false)
  (bind _returnvalue_2 (Effects|Components|ParticleSystem|SpawnFXAtLocation "/Game/Athena/Items/Consumables/Balloons/FXs/NS_Balloons_Pop.NS_Balloons_Pop" (Transformation|GetWorldLocation _balloon_skbudgeted)))
  (Effects|Components|ParticleSystem|SetColorParameter _returnvalue_2 "Balloon Color" (Variables|Default|GetBalloon_Color_Cache))
  (Utilities|FlowControl|Delay 3.0)
  (Actor|DestroyActor))

(fn ConstructionScript)
