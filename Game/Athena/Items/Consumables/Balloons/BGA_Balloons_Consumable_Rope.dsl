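; Balloon rope actor for the Balloons consumable that registers as the owning pawn's rope if none is set and drives balloon audio.
; BeginPlay caches ownership and team affiliation, choosing a friendly or hostile floating loop.
; Non-authority count and altitude events play inflate, release, start-to-rise and max-altitude cues.
; OnRep_bInAirWithABalloon fades the floating loop, adds rise audio for hostile owners, and toggles the pawn's looping sound indicator.
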(event EventBeginPlay
  ;[0 1 0.03968] Begin Play
  (|CallParentFunction)
  (bind _asfort_player_pawn (Utilities|Casting|CastToFortPlayerPawn (Actor|GetOwner)))
  (Variables|/Script/FortniteGame.FortPlayerPawn|GetBalloonRope _asfort_player_pawn
    (:then
      _)
    (:else
      (Balloon|SetBalloonRope _asfort_player_pawn self))) ; Set the players balloon rope to this if it is not set already
  (Variables|Default|SetOwnerIsLocal (Pawn|IsLocallyControlled _asfort_player_pawn))
  (bind _output_get (Variables|Default|SetOwnerIsHostileToLocal (Utilities|Enum|Equal(Enum) (Team|GetTeamAffiliationtoLocalPlayer (Team|GetActorTeam _asfort_player_pawn)) "Hostile")))
  (Audio|Components|Audio|SetSound (Variables|Default|GetBaloon_FloatingLoop_Cue) (select _output_get (Variables|Default|GetEnemyFloatingLoop) (Variables|Default|GetFriendlyFloatingLoop))))

(event Custom|BalloonCountAudioCheck
  (params (BalloonCount :type Integer) (WasShot :type Boolean) (MultiplesReleased :type Boolean) (MaxHeight :type Boolean))
  ;[0.5464 0.07286 1] Begin Flying Audio
  (Networking|SwitchHasAuthority
    (:Remote
      (Utilities|FlowControl|Sequence
        (:then_0
          (if WasShot
            _
            (elif MaxHeight
              _
              (elif (!= (Variables|Default|GetCurrentCount) BalloonCount)
                (Audio|SpawnSoundAttached (select (< (Variables|Default|GetCurrentCount) BalloonCount) "/Game/Sounds/Fort_Gadgets/Balloons/Balloon_Inflate_Cue.Balloon_Inflate_Cue" (select MultiplesReleased "/Game/Sounds/Fort_Gadgets/Balloons/Balloon_ReleaseMult_Cue.Balloon_ReleaseMult_Cue" "/Game/Sounds/Fort_Gadgets/Balloons/Balloon_Release_Cue.Balloon_Release_Cue")) (Variables|Default|GetSphere))
                (bind _output_get (Variables|Default|SetCurrentCount BalloonCount))
                (if (== _output_get 0)
                  (Audio|Components|Audio|FadeOut (Variables|Default|GetBalloon_StartToRise_Cue) 0.2 0.0 "Linear" :!class /Script/Engine.AudioComponent))
                (else
                  _)))))
        (:then_1
          (if (and (== BalloonCount (Variables|Default|GetBalloonsNeededForFlight)) (Variables|Default|GetOwnerIsLocal))
            (Components|Activation|Activate (Variables|Default|GetBalloon_StartToRise_Cue) false)))))))

(event Custom|BalloonMaxAltitudeAudioCheck
  (params (AtMax :type Boolean))
  ;[0.8338 0.2846 1] Max Altitude Cue
  (Networking|SwitchHasAuthority
    (:Remote
      (if (not AtMax)
        (Audio|Components|Audio|FadeOut (Variables|Default|GetBalloon_MaxAltitude_Cue) 0.5 0.0 "Linear" :!class /Script/Engine.AudioComponent))
      (Utilities|FlowControl|DoOnce false
        (:Completed
          (Audio|Components|Audio|FadeIn (Variables|Default|GetBalloon_MaxAltitude_Cue) 0.5 1.0 0.0 "Linear" :!class /Script/Engine.AudioComponent))))))

(fn OnRep_bInAirWithABalloon
  ;[1 1 1] Floating audio
  (Networking|SwitchHasAuthority
    (:Remote
      (if (Variables|Default|GetbInAirWithABalloon)
        (if (not (Audio|Components|Audio|IsPlaying (Variables|Default|GetBaloon_FloatingLoop_Cue)))
          (Utilities|FlowControl|Sequence
            (:then_0
              (Audio|Components|Audio|FadeIn (Variables|Default|GetBaloon_FloatingLoop_Cue) 0.25 1.0 0.0 "Linear" :!class /Script/Engine.AudioComponent)
              (if (Variables|Default|GetOwnerIsHostileToLocal)
                (Components|Activation|Activate (Variables|Default|GetBalloon_StartToRise_Cue) false)))
            (:then_1
              (bind _asplayer_pawn_athena (Utilities|Casting|CastToPlayerPawn_Athena (Actor|GetOwner)))
              (SoundIndicator|StartSoundIndicatorLooping (Variables|/Script/FortniteGame.FortPawn|GetSoundIndicatorComponent _asplayer_pawn_athena) (Transformation|GetActorLocation) (Variables|/Script/Engine.AudioComponent|GetSound (Variables|Default|GetBaloon_FloatingLoop_Cue)) 0.75 0))))
        (else
          (Utilities|FlowControl|Sequence
            (:then_0
              (Audio|Components|Audio|FadeOut (Variables|Default|GetBaloon_FloatingLoop_Cue) 0.25 0.0 "Linear" :!class /Script/Engine.AudioComponent)
              (Audio|Components|Audio|FadeOut (Variables|Default|GetBalloon_StartToRise_Cue) 0.25 0.0 "Linear" :!class /Script/Engine.AudioComponent))
            (:then_1
              (bind _asplayer_pawn_athena_2 (Utilities|Casting|CastToPlayerPawn_Athena (Actor|GetOwner)))
              (SoundIndicator|StopSoundIndicatorLooping (Variables|/Script/FortniteGame.FortPawn|GetSoundIndicatorComponent _asplayer_pawn_athena_2)))))))))

(fn ConstructionScript)
