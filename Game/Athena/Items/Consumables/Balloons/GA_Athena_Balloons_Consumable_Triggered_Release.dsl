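; Gameplay ability for the Athena Balloons consumable that handles an event-triggered release of the player's balloons.
; It can activate only while the avatar's FortPlayerPawn has at least one active balloon.
; On activation it calls the parent, sends the Balloons Deflate gameplay event, then grants and activates GA_BlockSlideUntilLand on the pawn, which by its name blocks sliding until landing.
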(event Ability|EventActivateAbilityFromEvent
  (params (EventData :type Gameplay%20Event%20Data%20Structure%20%28by%20ref%29))
  (|CallParentFunction EventData)
  (Ability|SendGameplayEvent "(TagName=\"Event.Athena.Balloons.Deflate\")")
  (bind _outfortplayerpawn (CallFunction|GetAvatarasValidFortPlayerPawn)
    (:ValidFortPlayerPawn
      (Ability|Advanced|GrantandActivateAbilityonSelf (Ability|GetFortAbilitySystemComponent _outfortplayerpawn) "/Game/Athena/Items/Parents/SlideBlock/GA_BlockSlideUntilLand.GA_BlockSlideUntilLand_C"))))

(fn CanActivateAbility
  (params (ActorInfo :type Gameplay%20Ability%20Actor%20Info%20Structure) (Handle :type Gameplay%20Ability%20Spec%20Handle%20Structure))
  (bind (_owneractor _avataractor) (Utilities|Struct|BreakGameplayAbilityActorInfo ActorInfo))
  (bind _asfort_player_pawn (Utilities|Casting|CastToFortPlayerPawn _avataractor))
  (return 0 (> (Balloon|GetNumActiveBalloons _asfort_player_pawn) 0)))
