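; Gameplay ability that releases all of a player's balloons when they perform certain actions, such as MME actions.
; A managing gameplay effect that tracks balloon-releasing tags activates it, avoiding per-action cancel logic.
; It can only activate while the avatar Fort player pawn has active balloons.
; On activation it calls the parent, sends the RemoveBalloons gameplay event, then immediately ends.
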(event Ability|EventActivateAbility
  (|CallParentFunction)
  ;[1 0.3955 0] Gets activated from a managing GE which tracks tags added to the player that should release balloons - more scalable than having every action cancel itself
  (Ability|SendGameplayEvent (Variables|Default|GetEventTag_RemoveBalloons)) ; remove all balloons from the player \nsince they are doing a MME action\nand should be consistent with other features\n(like Doom Gauntlets - FORT-766407)
  (Ability|EndAbility))

(fn CanActivateAbility
  (params (ActorInfo :type Gameplay%20Ability%20Actor%20Info%20Structure) (Handle :type Gameplay%20Ability%20Spec%20Handle%20Structure))
  (bind (_owneractor _avataractor) (Utilities|Struct|BreakGameplayAbilityActorInfo ActorInfo))
  (bind _asfort_player_pawn (Utilities|Casting|CastToFortPlayerPawn _avataractor))
  (return 0 (> (Balloon|GetNumActiveBalloons _asfort_player_pawn) 0)))
